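# Physics asset for the Pied Piper llama gadget prop, covering its root, offset, main and three floating llama joints.
# Each body is a single capsule of near-zero length, so effectively a sphere: large on root, mid-sized on the llamas, tiny on main and offset.
# Constraints with uniform swing and twist limit angles chain root to offset to main, which branches to the three llama bodies.
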PhysicsAsset /BRCosmetics/Animation/Props/Gadgets_PiedPiper_Llama/Meshes/PA_PiedPiper_Llama_Gadget
bodies {
 JNT_Llama_Float_MainMove_Llama01 [{shapeType=Capsule, length=0.505, radius=42.5249, rotation=<rot 0 -90 0>}]
 JNT_Llama_Float_MainMove_Llama02 [{shapeType=Capsule, length=0.505, radius=53.9404, rotation=<rot 0 -90 0>}]
 JNT_Llama_Float_MainMove_Llama03 [{shapeType=Capsule, length=0.505, radius=57.7925, rotation=<rot 0 -90 0>}]
 main [{shapeType=Capsule, length=0.505, radius=0.505, rotation=<rot 0 -90 0>}]
 offset [{shapeType=Capsule, length=0.505, radius=0.505, rotation=<rot 0 -90 0>}]
 root [{shapeType=Capsule, length=0.505, radius=131.6108, rotation=<rot 0 -90 0>}]
}
constraints [
 {bone1Name=offset, bone2Name=root, swing1LimitDegrees=45, swing2LimitDegrees=45, twistLimitDegrees=45}
 {bone1Name=main, bone2Name=offset, swing1LimitDegrees=45, swing2LimitDegrees=45, twistLimitDegrees=45}
 {bone1Name=JNT_Llama_Float_MainMove_Llama01, bone2Name=main, swing1LimitDegrees=45, swing2LimitDegrees=45, twistLimitDegrees=45}
 {bone1Name=JNT_Llama_Float_MainMove_Llama02, bone2Name=main, swing1LimitDegrees=45, swing2LimitDegrees=45, twistLimitDegrees=45}
 {bone1Name=JNT_Llama_Float_MainMove_Llama03, bone2Name=main, swing1LimitDegrees=45, swing2LimitDegrees=45, twistLimitDegrees=45}
]
