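# Animation Blueprint for the Pied Piper llama gadget, swinging three llamas as the owning actor turns.
# Each update turns the actor's yaw change into a clamped, interpolated, moving-average-smoothed value.
# That value drives yaw, slide and distance per llama, each chained from the previous with growing multipliers.
# The AnimGraph picks InMotion or DefaultSlot by bIsMoving2D, then adds those offsets to CoG bones in Llama01, 03, 02 order.
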AnimBlueprint /BRCosmetics/Animation/Props/Gadgets_PiedPiper_Llama/Meshes/PiedPiper_Llama_Gadget_Skeleton_AnimBlueprint
target_skeleton=&/PiedPiper_Llama_Gadget_Skeleton
graphs [
 Adjust Emote Llama Rotation {nodes [
   n1 = Adjust Emote Llama Rotation!() -> (then, Owning Actor:Actor Object Reference, CurrentLlamaRotation:double=~, DeltaTime:double=~, CurrentActorRotation:double=~) {is=title K2Node_FunctionEntry}
   n2 = Get Actor Rotation!(self:Actor Object Reference=@n1) -> (ReturnValue_Roll:float, ReturnValue_Pitch:float, ReturnValue_Yaw:float) {is=title K2Node_CallFunction}
   n3 = float * float(A=float - float(A=@n2.ReturnValue_Yaw, B=@n1.CurrentActorRotation) {is=title K2Node_CallFunction}, B=6)
   > Return Node!(
    execute
    NewActorRotation:double=@n2.ReturnValue_Yaw
    NewLlamaRotation:double=FInterpTo(Current=@n1.CurrentLlamaRotation, Target="Clamp (Float)"(Value=@n3, Min=-70, Max=70), DeltaTime=@n1.DeltaTime, InterpSpeed=6)
   )
  ]}
 AnimGraph {nodes [
   n1 = Slot 'DefaultSlot'()
   n2 = Slot 'InMotion'() {slotName=InMotion}
   n3 = Get bIsMoving2D()
   n4 = "Transform (Modify) Bone - Bone: JNT_CoG_Llama01"!(
    ComponentPose=Local To Component(LocalPose=Blend Poses by bool(BlendPose_0=@n2, BlendPose_1=@n1, bActiveValue=@n3) {blendType=Linear})
    Alpha=1
    Translation_X:double=Get Llama 1 Slide() -> (Llama 1 Slide:double)
    Translation_Y:double=Get Llama 1 Distance() -> (Llama 1 Distance:double)
    Translation_Z:double=0
    Rotation_Roll:float=0
    Rotation_Pitch:float=0
    Rotation_Yaw:float=Get Llama 1 Yaw() -> (Llama 1 Yaw:double)
    Scale="1,1,1"
   ) -> (Pose) {rotationMode=BMM_Additive, translationMode=BMM_Additive, alphaBoolBlend {blendOption=Linear}, boneToModify {boneName=JNT_CoG_Llama01}}
   n5 = "Transform (Modify) Bone - Bone: JNT_CoG_Llama03"!(
    ComponentPose=@n4
    Alpha=1
    Translation_X:double=Get Llama 2 Slide() -> (Llama 2 Slide:double)
    Translation_Y:double=Get Llama 2 Distance() -> (Llama 2 Distance:double)
    Translation_Z:double=0
    Rotation_Roll:float=0
    Rotation_Pitch:float=0
    Rotation_Yaw:float=Get Llama 2 Yaw() -> (Llama 2 Yaw:double)
    Scale="1,1,1"
   ) -> (Pose) {rotationMode=BMM_Additive, translationMode=BMM_Additive, alphaBoolBlend {blendOption=Linear}, boneToModify {boneName=JNT_CoG_Llama03}}
   n6 = "Transform (Modify) Bone - Bone: JNT_CoG_Llama02"!(
    ComponentPose=@n5
    Alpha=1
    Translation_X:double=Get Llama 3 Slide() -> (Llama 3 Slide:double)
    Translation_Y:double=Get Llama 3 Distance() -> (Llama 3 Distance:double)
    Translation_Z:double=0
    Rotation_Roll:float=0
    Rotation_Pitch:float=0
    Rotation_Yaw:float=Get Llama 3 Yaw() -> (Llama 3 Yaw:double)
    Scale="1,1,1"
   ) -> (Pose) {rotationMode=BMM_Additive, translationMode=BMM_Additive, alphaBoolBlend {blendOption=Linear}, boneToModify {boneName=JNT_CoG_Llama02}}
   Output Pose(Result=Component To Local(ComponentPose=@n6))
  ]}
 EventGraph {nodes [
   n1 = Event BlueprintUpdateAnimation()
   n2 = Get Current Llama Rotation() -> (Current Llama Rotation:double)
   > n3 = Adjust Emote Llama Rotation!(
    execute
    self
    Owning Actor:Actor Object Reference=Get Owning Actor() -> (Owning Actor:Actor Object Reference)
    CurrentLlamaRotation:double=@n2
    DeltaTime:double=@n1.DeltaTimeX
    CurrentActorRotation:double=Get Current Actor Rotation() -> (Current Actor Rotation:double)
   ) -> (then, NewActorRotation:double, NewLlamaRotation:double) {is=title K2Node_CallFunction}
   > Set Current Actor Rotation(Current Actor Rotation:double=@n3) -> (Output_Get:double)
   > n4 = Set Current Llama Rotation(Current Llama Rotation:double=Dynamic Weighted Moving Average Float!(CurrentSample:float=@n3.NewLlamaRotation, PreviousSample:float=@n2, MaxDistance:float=3, MinWeight:float=0.95, MaxWeight:float=0.3) -> (ReturnValue:float)) -> (Output_Get:double)
   > n5 = Sequence!(execute) -> (then_0, then_1, then_2:Exec)
   n6 = Set Llama 1 Distance(execute=@n5.then_2, Llama 1 Distance:double=float * float(A="Absolute (Float)"(A=@n4), B=1.3) {is=title K2Node_PromotableOperator}) -> (Output_Get:double)
   > n7 = Set Llama 2 Distance(Llama 2 Distance:double=float * float(A=@n6, B=2)) -> (Output_Get:double)
   > Set Llama 3 Distance(Llama 3 Distance:double=float * float(A=@n7, B=2)) -> (Output_Get:double)
   n8 = float * float(A=@n4, B=-1)
   n9 = Set Llama 1 Slide(execute=@n5.then_1, Llama 1 Slide:double=float * float(A=@n8, B=1.3)) -> (Output_Get:double)
   > n10 = Set Llama 2 Slide(Llama 2 Slide:double=float * float(A=@n9, B=2.4)) -> (Output_Get:double)
   > Set Llama 3 Slide(Llama 3 Slide:double=float * float(A=@n10, B=1.8)) -> (Output_Get:double)
   n11 = Set Llama 1 Yaw(execute=@n5.then_0, Llama 1 Yaw:double=float * float(A=@n8, B=0.6)) -> (Output_Get:double)
   > n12 = Set Llama 2 Yaw(Llama 2 Yaw:double=float * float(A=@n11, B=3)) -> (Output_Get:double)
   > Set Llama 3 Yaw(Llama 3 Yaw:double=float * float(A=@n12, B=1.8)) -> (Output_Get:double)
   Event BlueprintInitializeAnimation()
   > Set Owning Actor(Owning Actor:Actor Object Reference=GetOwningActor()) -> (Output_Get:Actor Object Reference)
   TryGetPawnOwner()
  ]}
]
variables [
 Current Actor Rotation {type=number}
 Current Llama Rotation {type=number}
 Llama 1 Distance {type=number}
 Llama 1 Slide {type=number}
 Llama 1 Yaw {type=number}
 Llama 2 Distance {type=number}
 Llama 2 Slide {type=number}
 Llama 2 Yaw {type=number}
 Llama 3 Distance {type=number}
 Llama 3 Slide {type=number}
 Llama 3 Yaw {type=number}
 Llama Distance {type=number}
 Llama Yaw {type=number}
 Owning Actor {type=Actor, default=null}
]
